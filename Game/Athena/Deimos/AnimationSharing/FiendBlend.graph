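# FiendBlend is an animation blueprint on the Husk V2 skeleton that crossfades between two skeletal mesh poses, apparently as the Fiend transition instance for animation sharing.
# Two Copy Pose From Mesh nodes read FromComponent and ToComponent, feeding a linear Blend Poses by bool driven by bBlendBool over BlendTime.
# It declares no variables of its own, so these inputs are presumably inherited from its parent class, and its event graph is empty.
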AnimBlueprint /Game/Athena/Deimos/AnimationSharing/FiendBlend
target_skeleton=/Game/Characters/Enemies/Husk/Mesh/Husk_V2/SK_Husk_V2_Skeleton
graphs [
 AnimGraph {nodes [
   n1 = Get ToComponent() -> (ToComponent:Skeletal Mesh Component Weak object)
   n2 = Get FromComponent() -> (FromComponent:Skeletal Mesh Component Weak object)
   n3 = Get BlendTime() -> (BlendTime:float)
   Output Pose(Result=Blend Poses by bool(
     BlendPose_0=Copy Pose From Mesh(SourceMeshComponent=@n2)
     BlendPose_1=Copy Pose From Mesh(SourceMeshComponent=@n1)
     BlendTime_0=@n3
     BlendTime_1=@n3
     bActiveValue=Get bBlendBool() -> (bBlendBool:bool)
    ) {blendType=Linear}) {layerGroup=None}
  ]}
 EventGraph {nodes [Event BlueprintBeginPlay() {is=title K2Node_Event}]}
]
variables []
