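# Animation Blueprint for the Fortnite player skeleton that layers an additive animation onto another mesh's pose, in the Deimos animation sharing folder.
# Copy Pose From Mesh takes the base pose from BaseComponent, and a Sequence Player plays AdditiveAnimation, looping according to bStateBool.
# Apply Mesh Space Additive combines both at full alpha for the output pose.
# The event graph holds only unconnected stub nodes, and the Blueprint declares no variables itself.
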AnimBlueprint /Game/Athena/Deimos/AnimationSharing/PlayerAdditive
target_skeleton=&Fortnite_M_Avg_Player_Skeleton
graphs [
 AnimGraph {nodes [
   n1 = Get BaseComponent() -> (BaseComponent:Skeletal Mesh Component Weak object)
   n2 = Get bStateBool() -> (bStateBool:bool)
   n3 = Get AdditiveAnimation() -> (AdditiveAnimation:Anim Sequence Weak object)
   Output Pose(Result=Apply Mesh Space Additive!(
     Base:Pose Link Structure=Copy Pose From Mesh(SourceMeshComponent=@n1)
     Additive:Pose Link Structure=Sequence Player!(Sequence:Anim Sequence Base Object Reference=@n3, bLoopAnimation:bool=@n2) -> (Pose)
     Alpha:float=1
    ) -> (Pose:Pose Link Structure) {is=title AnimGraphNode_ApplyMeshSpaceAdditive, alphaBoolBlend {blendOption=Linear}}) {layerGroup=None}
  ]}
 EventGraph {nodes [
   TryGetPawnOwner()
   Event BlueprintUpdateAnimation()
  ]}
]
variables []
