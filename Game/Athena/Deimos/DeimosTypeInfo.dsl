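; Deimos type info function library or class exposing a GetEnemyColor function for Deimos enemies.
; GetEnemyColor has no logic and always returns a constant zero, so it acts as a stub or default for an enemy color lookup.
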(fn GetEnemyColor
  (return 0))
