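# Translucent unlit material for drifting map fog or cloud layers, fading in as the camera gains height.
# Panning cloud noise, parallax-bumped above medium quality, is colored by time-of-day lit and ambient tints using a normal map and sky light direction.
# A noise sample offset toward the light fakes shadowing, and eye adaptation is divided out.
# Opacity combines noise, macro mask and depth fade; a WPO branch collapses vertices when the camera is below the actor.
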Material /Game/Athena/Deimos/M_MapFog
graph {
 properties {
  bEnableResponsiveAA=true
  bUseTranslucencyVertexFog=false
  blendMode=BLEND_Translucent
  pixelDepthOffsetMode=PDOM_Legacy
  shadingModel=MSM_Unlit
  usedShadingModels=Unlit
 }
 nodes [
  n1 = MaterialFunctionCall(
   Color_Morning=VectorParameter() {parameterName=LitMorning, defaultValue=<rgba 1 0.8701 0.7988 1>}.RGB
   Color_Day=VectorParameter() {parameterName=LitDay, defaultValue=<rgba 1 0.8701 0.7988 1>}.RGB
   Color_Evening=VectorParameter() {parameterName=LitEvening, defaultValue=<rgba 0.6927 0.323 0.2538 1>}.RGB
   Color_Night=VectorParameter() {parameterName=LitNight, defaultValue=<rgba 0.3132 0.3601 0.474 1>}.RGB
  ) {materialFunction=&Color_Time_of_Day}
  n2 = SkyAtmosphereLightDirection()
  n3 = Panner(
   Coordinate=Multiply(A=TextureCoordinate(), B=ScalarParameter() {defaultValue=1.5, parameterName=Tiling})
   Speed=AppendVector(A=ScalarParameter() {defaultValue=0.005, parameterName=PanSpeedU}, B=ScalarParameter() {defaultValue=0.005, parameterName=PanSpeedV})
  ) {bFractionalPart=true, speedX=0.005, speedY=0.005}
  n4 = TextureSample(UVs=@n3) {samplerType=SAMPLERTYPE_Normal, texture=/Game/Packages/Fortnite_UniqueMaterials_1/PostProcess/T_TilingClouds_02_N}
  n5 = Normalize(VectorInput=Transform(None=@n4.RGB))
  n6 = Saturate(None=ConstantBiasScale(None=DotProduct(A=@n2, B=@n5)))
  n7 = MaterialFunctionCall(
   Color_Morning=VectorParameter() {parameterName=AmbientMorning, defaultValue=<rgba 0.0378 0.0411 0.0573 1>}.RGB
   Color_Day=VectorParameter() {parameterName=AmbientDay, defaultValue=<rgba 0.1402 0.1854 0.2344 1>}.RGB
   Color_Evening=VectorParameter() {parameterName=AmbientEvening, defaultValue=<rgba 0.0378 0.0411 0.0573 1>}.RGB
   Color_Night=VectorParameter() {parameterName=AmbientNight, defaultValue=<rgba 0.0378 0.0411 0.0573 1>}.RGB
  ) {materialFunction=&Color_Time_of_Day}
  Add(A=Multiply(A=@n1.Result, B=@n6), B=Multiply(A=@n7.Result, B=@n4.B))
  n8 = LinearInterpolate(A=@n7.Result, B=@n1.Result, Alpha=@n6)
  n9 = TextureSampleParameter2D!(UVs=@n3, Bias=ScalarParameter() {defaultValue=4, parameterName=Mip}, Apply View MipBias) {mipValueMode=TMVM_MipBias, parameterName=TilingNoise, texture=/Game/Packages/Fortnite_UniqueMaterials_1/PostProcess/T_TilingClouds_02}
  n10 = ScalarParameter() {defaultValue=4, parameterName=Bump}
  n11 = QualitySwitch(Default=BumpOffset(Coordinate=@n3, Height=Multiply(A=@n9.R, B=@n10)), Low=@n3, Medium=@n3)
  n12 = PixelNormalWS()
  n13 = CameraVectorWS()
  n14 = Multiply(
   A=LinearInterpolate(
    A=ScalarParameter() {defaultValue=-0.076, parameterName=Low}
    B=ScalarParameter() {defaultValue=7.4647, parameterName=High}
    Alpha=TextureSampleParameter2D!(UVs=@n11, Apply View MipBias) {parameterName=TilingNoise, texture=&T_TilingClouds_02}.R
   )
   B=Saturate(None=DotProduct(A=@n12, B=@n13))
  )
  n15 = Multiply(A=@n7.Result, B=ScalarParameter() {defaultValue=0.11, parameterName=ShadowBrightness})
  n16 = ComponentMask(None=Transform(None=@n2) {transformSourceType=TRANSFORMSOURCE_World, transformType=TRANSFORM_Tangent}) {g=true, r=true}
  n17 = Add(A=@n3, B=Multiply(A=@n16, B=ScalarParameter() {defaultValue=-0.1024, parameterName=OffsetShadow}))
  n18 = LinearInterpolate(
   A=ScalarParameter() {defaultValue=-13.597, parameterName=Shadow_low}
   B=ScalarParameter() {defaultValue=3.3298, parameterName=Shadow}
   Alpha=OneMinus(None=TextureSampleParameter2D(UVs=@n17) {parameterName=TilingNoise, texture=&T_TilingClouds_02}.R)
  )
  Divide(A=Add(A=Multiply(A=@n8, B=@n14), B=Multiply(A=@n15, B=@n18)), B=EyeAdaptation().EyeAdaptation)
  n19 = ScalarParameter() {defaultValue=60000, parameterName=CameraFadeLength}
  n20 = ScalarParameter() {defaultValue=27980.0039, parameterName=CameraFadeOffset}
  MaterialFunctionCall(Fade Length=@n19, Fade Offset=@n20) {materialFunction=&CameraDepthFade}
  Multiply(A=@n16)
  n21 = TextureSampleParameter2D() {parameterName=MacroMask, samplerType=SAMPLERTYPE_LinearGrayscale, texture=/Game/Athena/Apollo/Environments/Blueprints/T_Cloud_MacroMask}
  n22 = Multiply(A=@n21.R, B=Add(A=@n14, B=@n18))
  n23 = ScalarParameter() {defaultValue=1, parameterName=Opacity}
  n24 = Subtract(A=ComponentMask(None=CameraPositionWS()) {b=true}, B=@n20)
  Multiply(
   A=DepthFade(Opacity=Multiply(A=@n22, B=@n23), FadeDistance=ScalarParameter() {defaultValue=9552.9561, parameterName=FadeDistance})
   B=Saturate(None=Divide(A=@n24, B=@n19))
  )
  Multiply(A=@n23, B=@n14)
  n25 = CameraPositionWS()
  n26 = ActorPositionWS()
  n27 = Constant()
  n28 = ActorPositionWS()
  n29 = WorldPosition()
  NamedRerouteDeclaration(None=If(A=ComponentMask(None=@n25) {b=true}, B=ComponentMask(None=@n26) {b=true}, "A > B"=@n27, "A == B"=@n27, "A < B"=Subtract(A=@n28, B=@n29.XYZ))) {name=WPO, nodeColor=<rgba 0 1 0.4141 1>}
  NamedRerouteUsage()
  TextureSample() {texture=/Game/TimeOfDay/TODM_Sequence/Materials/Sky/Clouds/T_Clouds_02_M}
  TextureSample() {samplerType=SAMPLERTYPE_LinearGrayscale, texture=/Game/Athena/Apollo/Environments/Blueprints/T_TwistyClouds_01}
 ]
}
